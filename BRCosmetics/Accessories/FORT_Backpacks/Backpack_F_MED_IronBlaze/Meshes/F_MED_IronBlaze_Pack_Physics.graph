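# Physics asset for the Iron Blaze backpack cosmetic, giving its hanging tassel simulated collision and motion.
# A large capsule covers the pack body, and four smaller capsules follow the tassel bone chain.
# Constraints link each tassel segment to the one above it, anchored at the pack, allowing moderate swing in both axes while twist is set to zero.
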PhysicsAsset /BRCosmetics/Accessories/FORT_Backpacks/Backpack_F_MED_IronBlaze/Meshes/F_MED_IronBlaze_Pack_Physics
bodies {
 dyn_tassle_01 [{shapeType=Capsule, length=5.25, radius=0.75, shapeName=dyn_tassle_01_capsule, center=<2.3207 0 0.204>, rotation=<rot -90 90 -90>}]
 dyn_tassle_02 [{shapeType=Capsule, length=5.25, radius=0.75, shapeName=dyn_tassle_02_capsule, center=<2.5758 0 0.204>, rotation=<rot -90 -179.6835 179.6835>}]
 dyn_tassle_03 [{shapeType=Capsule, length=5.25, radius=0.75, shapeName=dyn_tassle_03_capsule, center=<0.7558 0 0.204>, rotation=<rot -90 90 -90>}]
 dyn_tassle_04 [{shapeType=Capsule, length=2.5, radius=2.25, shapeName=dyn_tassle_04_capsule, center=<-0.8793 -0.7741 -0.705>}]
 pack [{shapeType=Capsule, length=10.25, radius=11.5, shapeName=pack_capsule, center=<0 1.1535 -6.668>}]
}
constraints [
 {bone1Name=dyn_tassle_01, bone2Name=pack, swing1LimitDegrees=50, swing2LimitDegrees=50, twistLimitDegrees=0}
 {bone1Name=dyn_tassle_02, bone2Name=dyn_tassle_01, swing1LimitDegrees=50, swing2LimitDegrees=50, twistLimitDegrees=0}
 {bone1Name=dyn_tassle_03, bone2Name=dyn_tassle_02, swing1LimitDegrees=50, swing2LimitDegrees=50, twistLimitDegrees=0}
 {bone1Name=dyn_tassle_04, bone2Name=dyn_tassle_03, swing1LimitDegrees=50, swing2LimitDegrees=50, twistLimitDegrees=0}
]
